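# WelderBot_ArmMove_Cue is a Sound Cue for the arm movement of the WelderBot, filed under supply drop drone sounds.
# A Random node picks one of four equally weighted arm-move waves and feeds it through a Modulator for playback variation.
# The cue uses the WelderBot attenuation settings and the Fort_Gadgets sound class, with a reduced volume multiplier.
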SoundCue /Figment_S09/Athena/Sounds/SupplyDrops/Drone/WelderBot/WelderBot_ArmMove_Cue
root=Modulator_3
nodes [
 {id=Modulator_3, children [Random_3]}
 {id=Random_3, children [WavePlayer_5, WavePlayer_6, WavePlayer_13, WavePlayer_14], properties {weights [1,1,1,1]}}
 {id=WavePlayer_13, properties {soundWaveAssetPtr=&/WelderBot_Arm_Move_02}}
 {id=WavePlayer_14, properties {soundWaveAssetPtr=&/WelderBot_Arm_Move_01}}
 {id=WavePlayer_5, properties {soundWaveAssetPtr=&/WelderBot_Arm_Move_04}}
 {id=WavePlayer_6, properties {soundWaveAssetPtr=&/WelderBot_Arm_Move_03}}
]
properties {
 attenuationSettings=&/WelderBot_Att
 maxDistance=12750
 soundClassObject=&Fort_Gadgets
 volumeMultiplier=0.6
}
