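# WelderBot_Light_Blink_Cue is a SoundCue for the status light flash of the WelderBot supply drop drone.
# A Random node picks between two flash wave variations with equal weight.
# The cue plays through the WelderBot attenuation settings in the Fort_Gadgets sound class at reduced volume.
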SoundCue /Figment_S09/Athena/Sounds/SupplyDrops/Drone/WelderBot/WelderBot_Light_Blink_Cue
root=Random_3
nodes [
 {id=Random_3, children [WavePlayer_5,WavePlayer_6], properties {weights [1,1]}}
 {id=WavePlayer_5, properties {soundWaveAssetPtr=&/WelderBot_StatusLight_Flash_01}}
 {id=WavePlayer_6, properties {soundWaveAssetPtr=&/WelderBot_StatusLight_Flash_02}}
]
properties {
 attenuationSettings=&/WelderBot_Att
 maxDistance=12750
 soundClassObject=&Fort_Gadgets
 volumeMultiplier=0.6
}
