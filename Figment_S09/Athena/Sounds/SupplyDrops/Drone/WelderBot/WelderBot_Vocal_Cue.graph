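# Sound cue for the WelderBot's vocal chirps, built from supply drop drone beeps and played as a gadget sound with WelderBot attenuation.
# A quality level node selects one of three branches, each picking randomly among equally weighted beeps and raising pitch by a random amount.
# The first slot draws on six base-game beeps, the second on three optimized copies, and the last two share a branch of two optimized copies.
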SoundCue /Figment_S09/Athena/Sounds/SupplyDrops/Drone/WelderBot/WelderBot_Vocal_Cue
root=QualityLevel_0
nodes [
 {id=Modulator_0, children [Random_0], properties {pitchMax=1.3, pitchMin=1}}
 {id=Modulator_1, children [Random_1], properties {pitchMax=1.3, pitchMin=1}}
 {id=Modulator_2, children [Random_2], properties {pitchMax=1.3, pitchMin=1}}
 {id=QualityLevel_0, children [Modulator_0,Modulator_1,Modulator_2,Modulator_2]}
 {
  id=Random_0
  children [WavePlayer_7, WavePlayer_8, WavePlayer_9, WavePlayer_10, WavePlayer_11, WavePlayer_12]
  properties {weights [1,1,1,1,1,1]}
 }
 {id=Random_1, children [WavePlayer_0,WavePlayer_1,WavePlayer_2], properties {weights [1,1,1]}}
 {id=Random_2, children [WavePlayer_3,WavePlayer_4], properties {weights [1,1]}}
 {id=WavePlayer_0, properties {soundWaveAssetPtr=/Figment_S09/Athena/Sounds/SupplyDrops/Drone/OPT/Athena_SupplyDrop_Drone_Beep_03_Opt}}
 {id=WavePlayer_1, properties {soundWaveAssetPtr=/Figment_S09/Athena/Sounds/SupplyDrops/Drone/OPT/Athena_SupplyDrop_Drone_Beep_05_Opt}}
 {id=WavePlayer_10, properties {soundWaveAssetPtr=/Game/Athena/Sounds/SupplyDrops/Drone/Athena_SupplyDrop_Drone_Beep_04}}
 {id=WavePlayer_11, properties {soundWaveAssetPtr=/Game/Athena/Sounds/SupplyDrops/Drone/Athena_SupplyDrop_Drone_Beep_05}}
 {id=WavePlayer_12, properties {soundWaveAssetPtr=/Game/Athena/Sounds/SupplyDrops/Drone/Athena_SupplyDrop_Drone_Beep_06}}
 {id=WavePlayer_2, properties {soundWaveAssetPtr=/Figment_S09/Athena/Sounds/SupplyDrops/Drone/OPT/Athena_SupplyDrop_Drone_Beep_06_Opt}}
 {id=WavePlayer_3, properties {soundWaveAssetPtr=&Athena_SupplyDrop_Drone_Beep_06_Opt}}
 {id=WavePlayer_4, properties {soundWaveAssetPtr=&Athena_SupplyDrop_Drone_Beep_05_Opt}}
 {id=WavePlayer_7, properties {soundWaveAssetPtr=/Game/Athena/Sounds/SupplyDrops/Drone/Athena_SupplyDrop_Drone_Beep_01}}
 {id=WavePlayer_8, properties {soundWaveAssetPtr=/Game/Athena/Sounds/SupplyDrops/Drone/Athena_SupplyDrop_Drone_Beep_02}}
 {id=WavePlayer_9, properties {soundWaveAssetPtr=/Game/Athena/Sounds/SupplyDrops/Drone/Athena_SupplyDrop_Drone_Beep_03}}
]
properties {
 attenuationSettings=&/WelderBot_Att
 maxDistance=12750
 soundClassObject=&Fort_Gadgets
 volumeMultiplier=0.6
}
