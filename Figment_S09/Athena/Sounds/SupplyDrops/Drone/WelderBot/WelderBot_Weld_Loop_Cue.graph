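# Sound cue for the looping weld sound of the supply drop drone's WelderBot.
# A single looping wave player plays the WelderBot_Weld_Loop_01 wave, with no randomization or modulation.
# It uses the WelderBot attenuation asset and the Fort_Gadgets sound class.
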SoundCue /Figment_S09/Athena/Sounds/SupplyDrops/Drone/WelderBot/WelderBot_Weld_Loop_Cue
root=WavePlayer_5
nodes [{id=WavePlayer_5, properties {bLooping=true, soundWaveAssetPtr=&/WelderBot_Weld_Loop_01}}]
properties {
 attenuationSettings=&/WelderBot_Att
 maxDistance=12750
 soundClassObject=&Fort_Gadgets
 volumeMultiplier=1
}
