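# Sound cue for the stop of welding by the WelderBot, part of the supply drop drone audio.
# A Random node picks between two equally weighted weld-stop wave variations and feeds a Modulator left at its default settings.
# The cue uses the shared WelderBot attenuation asset and plays in the Fort_Gadgets sound class.
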SoundCue /Figment_S09/Athena/Sounds/SupplyDrops/Drone/WelderBot/WelderBot_Weld_Stop_Cue
root=Modulator_3
nodes [
 {id=Modulator_3, children [Random_0]}
 {id=Random_0, children [WavePlayer_0,WavePlayer_1], properties {weights [1,1]}}
 {id=WavePlayer_0, properties {soundWaveAssetPtr=&/WelderBot_Weld_Stop_01}}
 {id=WavePlayer_1, properties {soundWaveAssetPtr=&/WelderBot_Weld_Stop_02}}
]
properties {
 attenuationSettings=&/WelderBot_Att
 maxDistance=12750
 soundClassObject=&Fort_Gadgets
 volumeMultiplier=1
}
